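# Item-layer Anim Blueprint for the Captain America (Ashton) shield on the Fortnite player skeleton, filling Item Override slots.
# Jump and fall layers pick shield additives by bIsTargeting, while jetpack and zipline layers reuse assault-rifle additives.
# The final-pose layer adds a ShieldHit stagger state machine onto a TacJump machine reading the cached pose.
# TacJump's block-pose bone blend is unconnected, and many other layers are unwired stubs.
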AnimBlueprint /AshtonGameplay/Assets/CaptainAmerica/Shield/Anim/Ashton_Shield_AnimBP
target_skeleton=&Fortnite_M_Avg_Player_Skeleton
graphs [
 AnimGraph {nodes [Output Pose() {layerGroup=None}]}
 EventGraph {links [], nodes []}
 ItemAimOffset {nodes [
   Input Pose!() -> (Pose, AimOffsetAlpha:double, Yaw:double, Pitch:double) {name=InPose_AimOffset}
   Input Pose!() -> (Pose) {name=InPose_PreAimOffsets}
   Output Pose() {layerGroup=Item Override}
  ]}
 ItemFallAdditive {nodes [
   Input Pose() {name=InPoseItemFallAdditive}
   Output Pose(Result=Blend Poses by bool(
     BlendPose_0="Sequence Player (additive) 'Ashton_Shield_JumpApex_Loop_CMM_Additive'"() {bLoopAnimation=false, sequence=&/Ashton_Shield_JumpApex_Loop_CMM_Additive}
     BlendPose_1="Sequence Player (additive) 'Ashton_Shield_JumpFall_CMM_Additive'"() {bLoopAnimation=false, sequence=&/Ashton_Shield_JumpFall_CMM_Additive}
     bActiveValue=Get bIsTargeting()
    ) {blendType=Linear}) {layerGroup=Item Override}
   "Sequence Player (additive) 'AshtonChicago_JumpFall_Additive_CMM'"!(PlayRate:float=Get FallPlayRate() -> (FallPlayRate:float)) -> (Pose) {
    bLoopAnimation=false
    sequence=/Game/Animation/Game/MainPlayer/Locomotion/Medium/Male/Jump/Ashton/Chicago/AshtonChicago_JumpFall_Additive_CMM
   }
  ]}
 ItemFallLandAdditive {nodes [
   Output Pose(Result=Input Pose() {name=InPoseItemFallLandAdditive}) {layerGroup=Item Override}
   "Sequence Player (additive) 'AshtonChicago_JumpLand_Additive_CMM'"() {
    groupName=JumpLand
    method=SyncGroup
    sequence=/Game/Animation/Game/MainPlayer/Locomotion/Medium/Male/Jump/Ashton/Chicago/AshtonChicago_JumpLand_Additive_CMM
   }
  ]}
 ItemFallLevel2Override {nodes [
   Input Pose() {name=InPose_FallLevel2}
   Output Pose() {layerGroup=Item Override}
  ]}
 ItemFinalPoseOverride {nodes [
   AnimGraphNode_StateMachine_0 = TacJump SM() {is=title AnimGraphNode_StateMachine}
   AnimGraphNode_StateMachine_1 = ShieldHit SM() {is=title AnimGraphNode_StateMachine}
   Output Pose(Result=Apply Additive(Base=@AnimGraphNode_StateMachine_0, Additive=@AnimGraphNode_StateMachine_1) {alphaBoolBlend {blendOption=Linear}}) {layerGroup=Item Override}
   Save cached pose 'Final Pose'(Pose=Input Pose() {name=InFinalPose}) {cacheName=Final Pose}
  ]}
 TacJump SM {
  entry=Wait
  path=ItemFinalPoseOverride.AnimGraphNode_StateMachine_0.TacJump SM
  ids=id
  nodes [
   AnimStateNode_0 = Wait() {is=title AnimStateNode}
   AnimStateNode_1 = TacSprintJumping() {is=title AnimStateNode}
   AnimStateTransitionNode_0 = AnimStateTransitionNode() {crossfadeDuration=0.05}
   AnimStateTransitionNode_1 = AnimStateTransitionNode() {crossfadeDuration=0.4}
  ]
  links [
   {
    dst=Wait
    rule=ItemFinalPoseOverride.AnimGraphNode_StateMachine_0.TacJump SM.AnimStateTransitionNode_1.Transition
    src=TacSprintJumping
    via=AnimStateTransitionNode_1
   }
   {
    dst=TacSprintJumping
    rule=ItemFinalPoseOverride.AnimGraphNode_StateMachine_0.TacJump SM.AnimStateTransitionNode_0.Transition
    src=Wait
    via=AnimStateTransitionNode_0
   }
  ]
 }
 Wait {
  path=ItemFinalPoseOverride.AnimGraphNode_StateMachine_0.TacJump SM.AnimStateNode_0.Wait
  nodes [Output Animation Pose(Result=Use cached pose 'Final Pose'())]
 }
 TacSprintJumping {
  path=ItemFinalPoseOverride.AnimGraphNode_StateMachine_0.TacJump SM.AnimStateNode_1.TacSprintJumping
  nodes [
   Layered blend per bone(
    BlendPoses_0=Sequence Player 'Ashton_Shield_Pose_Block_CMM'() {bLoopAnimation=false, sequence=&/Ashton_Shield_Pose_Block_CMM}
    BlendWeights_0=0.4
   ) {bMeshSpaceRotationBlend=true, layerSetup [{branchFilters [{blendDepth=3, boneName=spine_04}, {blendDepth=-1, boneName=clavicle_l}]}]}
   Output Animation Pose(Result=Use cached pose 'Final Pose'())
  ]
 }
 Transition {
  path=ItemFinalPoseOverride.AnimGraphNode_StateMachine_0.TacJump SM.AnimStateTransitionNode_0.Transition
  nodes [Result(bCanEnterTransition=Get bIsTacticalJump() -> (bIsTacticalJump:bool))]
 }
 Transition {
  path=ItemFinalPoseOverride.AnimGraphNode_StateMachine_0.TacJump SM.AnimStateTransitionNode_1.Transition
  nodes [Result(bCanEnterTransition=NOT Boolean(A=Get bIsTacticalJump()))]
 }
 ShieldHit SM {
  entry=Wait
  path=ItemFinalPoseOverride.AnimGraphNode_StateMachine_1.ShieldHit SM
  ids=id
  nodes [
   AnimStateNode_0 = Wait()
   AnimStateNode_1 = Shield Hit() {is=title AnimStateNode}
   AnimStateTransitionNode_2 = AnimStateTransitionNode() {crossfadeDuration=0.05}
   AnimStateTransitionNode_3 = AnimStateTransitionNode() {bAutomaticRuleBasedOnSequencePlayerInState=true, crossfadeDuration=0.15}
  ]
  links [
   {
    dst=Wait
    rule=ItemFinalPoseOverride.AnimGraphNode_StateMachine_1.ShieldHit SM.AnimStateTransitionNode_3.Transition
    src=Shield Hit
    via=AnimStateTransitionNode_3
   }
   {
    dst=Shield Hit
    rule=ItemFinalPoseOverride.AnimGraphNode_StateMachine_1.ShieldHit SM.AnimStateTransitionNode_2.Transition
    src=Wait
    via=AnimStateTransitionNode_2
   }
  ]
 }
 Wait {path=ItemFinalPoseOverride.AnimGraphNode_StateMachine_1.ShieldHit SM.AnimStateNode_0.Wait, nodes [Output Animation Pose()]}
 Shield Hit {
  path=ItemFinalPoseOverride.AnimGraphNode_StateMachine_1.ShieldHit SM.AnimStateNode_1.Shield Hit
  nodes [Output Animation Pose(Result="Sequence Player (additive) 'Ashton_Shield_StaggerBuildup'"() {bLoopAnimation=false, sequence=&/Ashton_Shield_StaggerBuildup})]
 }
 Transition {
  path=ItemFinalPoseOverride.AnimGraphNode_StateMachine_1.ShieldHit SM.AnimStateTransitionNode_2.Transition
  nodes [Result(bCanEnterTransition=Get bIsShieldHit() -> (bIsShieldHit:bool))]
 }
 Transition {path=ItemFinalPoseOverride.AnimGraphNode_StateMachine_1.ShieldHit SM.AnimStateTransitionNode_3.Transition, nodes [Result()]}
 ItemFullBodyAimOffsets {nodes [
   Input Pose!() -> (Pose, FullBody_AimOffsetAlpha:double, Aim_Yaw:double, Aim_Pitch:double) {name=InPoseFullBodyAimOffsets}
   Input Pose!() -> (Pose) {name=InPosePreFullBodyAimOffsets}
   Output Pose() {layerGroup=Item Override}
  ]}
 ItemFullBodyOverride {nodes [
   Input Pose() {name=InPoseFullBody}
   Output Pose() {layerGroup=Item Override}
  ]}
 ItemIdleOverride {nodes [
   Input Pose() {name=InPose_Idle}
   Output Pose() {layerGroup=Item Override}
  ]}
 ItemJetPackJumpAdditive {nodes [Output Pose(Result="Sequence Player (additive) 'JetpackApex_ARDS_Loop_Additive'"() {sequence=/Game/Animation/Game/MainPlayer/Locomotion/Medium/Male/Jump/AssaultRifle/JetpackApex_ARDS_Loop_Additive}) {layerGroup=Item Override}]}
 ItemJetPackStartAdditive {nodes [Output Pose(Result="Sequence Player (additive) 'JetpackUp_ARDS_Additive'"() {bLoopAnimation=false, sequence=/Game/Animation/Game/MainPlayer/Locomotion/Medium/Male/Jump/AssaultRifle/JetpackUp_ARDS_Additive}) {layerGroup=Item Override}]}
 ItemJumpLoopAdditive {nodes [Output Pose(Result=Blend Poses by bool(
     BlendPose_0="Sequence Player (additive) 'Ashton_Shield_JumpApex_Loop_CMM_Additive'"() {bLoopAnimation=false, sequence=&Ashton_Shield_JumpApex_Loop_CMM_Additive}
     BlendPose_1=Input Pose() {name=InPoseItemJumpLoopAdditive}
     bActiveValue=Get bIsTargeting()
    ) {blendType=Linear}) {layerGroup=Item Override}]}
 ItemJumpOverride {nodes [
   Input Pose() {name=InPoseJump}
   Output Pose() {layerGroup=Item Override}
  ]}
 ItemJumpUpAdditive {nodes [
   Output Pose(Result=Blend Poses by bool(
     BlendPose_0="Sequence Player (additive) 'Ashton_Shield_JumpApex_Loop_CMM_Additive'"() {bLoopAnimation=false, sequence=&Ashton_Shield_JumpApex_Loop_CMM_Additive}
     BlendPose_1=Input Pose() {name=InPoseItemJumpUpAdditive}
     bActiveValue=Get bIsTargeting()
    ) {blendType=Linear}) {layerGroup=Item Override}
   "Sequence Player (additive) 'AshtonChicago_JumpUp_Additive_CMM'"() {bLoopAnimation=false, sequence=/Game/Animation/Game/MainPlayer/Locomotion/Medium/Male/Jump/Ashton/Chicago/AshtonChicago_JumpUp_Additive_CMM}
  ]}
 ItemPreAimOffsets {nodes [
   Input Pose() {name=InPosePreAimOffsets}
   Output Pose() {layerGroup=Item Override}
  ]}
 ItemPreIK {nodes [
   Input Pose() {name=InPose_PreIK}
   Output Pose() {layerGroup=Item Override}
  ]}
 ItemRightHandIK {nodes [
   Input Pose() {name=InPosePreRightHandIK}
   Input Pose() {name=InPosePostRightHandIK}
   Output Pose() {layerGroup=Item Override}
  ]}
 ItemSkeletalControl {nodes [
   Input Pose()
   Output Pose() {layerGroup=Item Override}
  ]}
 ItemSwimJumpFallLoop {nodes [Output Pose() {layerGroup=Item Override}]}
 ItemSwimJumpSurfaceEnd {nodes [Output Pose() {layerGroup=Item Override}]}
 ItemSwimJumpSurfaceLoop {nodes [Output Pose() {layerGroup=Item Override}]}
 ItemUpperBody {nodes [
   Input Pose!() -> (Pose, InputParam:Fort Anim Input Adjusted Aim Structure) {name=InPoseUpperBody}
   Output Pose() {layerGroup=Item Override}
  ]}
 ItemZipLineStartAdditive {nodes [Output Pose(Result="Sequence Player (additive) 'ZipLine_Entry_ARDS_Additive'"() {bLoopAnimation=false, sequence=/Game/Animation/Game/MainPlayer/Combat/Gadgets/Medium/Male/ZipLine/ZipLine_Entry_ARDS_Additive}) {layerGroup=Item Override}]}
]
variables []
