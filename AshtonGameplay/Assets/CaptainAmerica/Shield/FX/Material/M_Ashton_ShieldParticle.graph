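# Unlit, translucent, two-sided material for Captain America shield effects on Niagara mesh particles.
# Color blends the shield texture toward white, multiplies it by particle color, and passes it through an emissive adjustment function.
# An alpha term combines particle alpha, optional texture alpha via static switch, and a Fresnel edge falloff blended by a scalar parameter.
# A separate inverted-Fresnel branch feeds nothing visible in the graph.
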Material /AshtonGameplay/Assets/CaptainAmerica/Shield/FX/Material/M_Ashton_ShieldParticle
graph {
 properties {
  bAutomaticallySetUsageInEditor=false
  bCastRayTracedShadows=false
  bUsedWithNiagaraMeshParticles=true
  blendMode=BLEND_Translucent
  previewMesh=/AshtonGameplay/Assets/CaptainAmerica/Shield/FX/Meshes/SM_AshtonChicago_ShieldParticle
  shadingModel=MSM_Unlit
  twoSided=true
  usedShadingModels=Unlit
 }
 nodes [
  n1 = TextureSampleParameter2D() {parameterName=ShieldTexture, texture=/Game/Athena/Items/Gameplay/BackPacks/Ashton/Chicago/Textures/T_Candy_Apple_Backpack_D}
  n2 = ScalarParameter() {defaultValue=0.25, parameterName=White_Interpolate}
  n3 = ParticleColor()
  MaterialFunctionCall(Emissive=Multiply(A=LinearInterpolate(A=@n1.RGB, Alpha=@n2) {constAlpha=0.25}, B=@n3.RGB)) {materialFunction=&MF_FX_EmissiveAdjustments}
  n4 = Fresnel() {baseReflectFraction=0, exponent=8}
  n5 = Multiply(A=@n3.A, B=@n4)
  n6 = ScalarParameter() {parameterName=Use_Fresnel}
  Multiply(
   A=StaticSwitchParameter(True=@n1.A, False=Constant() {r=1}) {parameterName=Use_Texture_Alpha}
   B=Multiply(A=@n3.A, B=LinearInterpolate(A=@n3.A, B=@n5, Alpha=@n6))
  )
  Power(Base=OneMinus(None=@n4)) {constExponent=1}
 ]
}
