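# Unlit translucent Niagara particle material for a screen-edge glow lens effect in the Captain America shield FX.
# A noise texture panned in radial coordinates slightly distorts the UVs to animate the mask shape.
# An inverted, exponent-adjusted sphere mask clears the center and is multiplied by an inverted square edge mask.
# The combined mask is tinted by a Color parameter and particle color, with particle alpha scaling opacity.
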Material /AshtonGameplay/Assets/CaptainAmerica/Shield/FX/Material/M_Ashton_Shield_Lens_ScreenEdgeGlow
graph {
 properties {
  bUsedWithNiagaraMeshParticles=true
  bUsedWithNiagaraSprites=true
  blendMode=BLEND_Translucent
  pixelDepthOffsetMode=PDOM_Legacy
  shadingModel=MSM_Unlit
  usedShadingModels=Unlit
 }
 nodes [
  n1 = Power(Base=SphereMask(
    A=NamedRerouteUsage()
    B=Constant() {r=0.5}
    Radius=ScalarParameter() {defaultValue=0.3, parameterName=Radius}
    Hardness=ScalarParameter() {defaultValue=0.85, parameterName=Hardness}
   ) {attenuationRadius=0.5, hardnessPercent=0})
  n2 = NamedRerouteUsage()
  n3 = Constant() {r=1.5}
  n4 = Constant() {r=1}
  n5 = Multiply(
   A=Power(Base=OneMinus(None=@n1), Exp=ScalarParameter() {defaultValue=1.5, parameterName=Exponent})
   B=OneMinus(None=MaterialFunctionCall!(UVs=@n2, Mult=@n3, Exp=@n4) {materialFunction=/Game/Effects/Fort_Effects/MaterialFunctions/MF_SquareEdgeMask}.SquareMask)
  )
  n6 = ParticleColor()
  Multiply(A=Multiply(A=@n5), B=@n6.A)
  Multiply(A=VectorParameter() {parameterName=Color, defaultValue=<rgba 1 1 1 0>}.RGB, B=Multiply(A=@n6.RGB, B=@n5))
  n7 = TextureCoordinate()
  n8 = Multiply(
   A=Panner(Coordinate=MaterialFunctionCall() {materialFunction=&VectorToRadialValue}.Radial Coordinates) {speedY=-0.5}
   B=Constant2Vector() {g=0.05, r=2}
  )
  n9 = Subtract(A=ComponentMask(None=TextureSample(UVs=@n8) {samplerType=SAMPLERTYPE_LinearColor, texture=&HighContrastNoise}.RGB) {g=true, r=true}) {constB=0.5}
  NamedRerouteDeclaration(None=Add(A=@n7, B=Multiply(A=@n9) {constB=0.05})) {name=UVs, nodeColor=<rgba 1 0.7266 0 1>}
  Reroute(None=@n7)
 ]
}
