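# Niagara system for the spark burst when Captain America's shield blocks a hit.
# Two sprite emitters burst once: a few Rico_bullets streaks and a denser, shorter-lived Small_rico spray.
# System-level impact intensity and side-angle variables set color and cone direction; a User.ImpactStage input is exposed.
# Drag and gravity shape the motion, speed-based scaling stretches sprites, and color curves cool them from white-hot to orange.
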NiagaraSystem /AshtonGameplay/Assets/CaptainAmerica/Shield/FX/NS_Ashton_Shield_Block_Impact
dependencies {
 usedDataInterfaces [&NiagaraDataInterfaceCurve, &NiagaraDataInterfaceColorCurve, &NiagaraDataInterfaceArrayFloat]
 usedDynamicInputs [&Lerp_Vector2, &SelectFloatFromArray]
 usedModules [
  &self:SystemSpawnScript.SystemScriptSource.SystemScriptGraph.NiagaraNodeAssignment_1.SetVariables_#1
  &self:SystemSpawnScript.SystemScriptSource.SystemScriptGraph.NiagaraNodeAssignment_2.SetVariables_#2
  &self:SystemSpawnScript.SystemScriptSource.SystemScriptGraph.NiagaraNodeAssignment_0.SetVariables_#3
  &SystemState
 ]
}
emitters {
 Rico_bullets {
  emitterUpdateScript [
   &EmitterState()
   &SpawnBurst_Instantaneous(Spawn Count=3, Use Spawn Probability=false)
  ]
  particleSpawnScript [
   &InitializeParticle(
    Color=&LerpLinearColors(StartColor=<rgba 1.5 0.4313 0.075 1>, EndColor=<rgba 3 1.8744 0.4518 1>, LerpFactor=$System.Impactintencity)
    Lifetime Max=0.3
    Lifetime Min=0.2
    Lifetime Mode=::NewEnumerator1
    Sprite Size Mode=::NewEnumerator3
    Uniform Sprite Size=&RandomRangeFloat(Minimum=2, Maximum=4)
    UsePositionOffset=false
    Write Material Random=true
   ) {module=InitializeParticle001}
   /SetVariables(Particles.Position=&AddVectorToPosition(
     Position=$Particles.Position
     Vector=&RotateVectorByQuaternion(Vector To Rotate=<8 0 0>, Quaternion=$Engine.Owner.Rotation)
    )) {isSetParameters=true}
   &AddVelocityInCone(
    Cone Angle=30
    Cone Axis=/Niagara/DynamicInputs/Vectors/RotateVector(
     Vector To Rotate:Vector3f=&MakeVector(X=0.45, Y=&Multiply_Float(A=0.3, B=$System.ImpactSideAngleVariation), Z=0.9)
     Yaw:NiagaraFloat=0
     Pitch:NiagaraFloat=0
     Roll:NiagaraFloat=&Multiply_Float(A=15, B=&RandomRangeFloat(Minimum=-1, Maximum=1))
     Rotation Rate:NiagaraFloat=1
    )
    Module.Use Velocity Falloff On Cone Axis=false
    Velocity Strength=&Multiply_Float(A=&UniformRangedFloat(Minimum=150, Maximum=270), B=10)
   )
   &InheritVelocity(Inherited Velocity Speed Limit=2000, Speed Limit=true, UseSpeedThreshold=false) {enabled=false}
   &CameraOffset(Camera Offset Amount=-5)
  ]
  particleUpdateScript [
   &ParticleState()
   &InheritSourceMovement(
    Applied Position Scale=&VectorFromFloat(Value=&FloatFromCurve(FloatCurve=DataInterface() {propertyValues {Curve {keys [
          {arriveTangent=-3.3349, interpMode=RCIM_Cubic, leaveTangent=-3.3349, tangentMode=RCTM_User, time=0, value=1}
          {interpMode=RCIM_Cubic, tangentMode=RCTM_User, time=0.5, value=0}
         ]}}}))
    ApplyForce=false
    ApplyPosition=true
    ApplyPositionFalloff=false
   )
   &ScaleSpriteSize(Uniform Curve Sprite Scale=DataInterface() {propertyValues {Curve {keys [{interpMode=RCIM_Cubic, tangentMode=RCTM_User, time=0, value=0}, {interpMode=RCIM_Cubic, tangentMode=RCTM_User, time=0.1, value=1}]}}})
   &ScaleColor(
    Scale Alpha=&FloatFromCurve(FloatCurve=DataInterface() {propertyValues {Curve {keys [
         {tangentMode=RCTM_User, time=0, value=1}
         {arriveTangent=-0.0405, interpMode=RCIM_Cubic, leaveTangent=-0.0405, tangentMode=RCTM_User, time=0.7009, value=1}
         {arriveTangent=-8.0089, interpMode=RCIM_Cubic, leaveTangent=-8.0089, tangentMode=RCTM_User, time=1, value=0}
        ]}}})
    ScaleA=true
    ScaleRGB=true
   )
   &Drag(Drag=&RandomRangeFloat(Minimum=20, Maximum=25), Use Linear Drag=true, Use Rotational Drag=true)
   &GravityForce(Gravity=<0 0 -1960>)
   &SolveForcesAndVelocity(Clamp Velocity=false, Limit Acceleration=false)
   &ScaleSpriteSizeBySpeed(
    Max Scale Factor=&RandomRangeVector2D(Minimum=<0.5 10>, Maximum=<0.6 15>)
    Min Scale Factor=<0.35 0.5>
    Velocity Threshold=1800
   )
   &DynamicMaterialParameters(
    Index 0 Param 1=1
    Index 0 Param 2=1
    Index 0 Param 3=&FloatFromCurve(FloatCurve=DataInterface() {propertyValues {Curve {keys [<at 0 1 RCIM_Cubic>,<at 1 0 RCIM_Cubic>]}}})
    Index 0 Param 4=1
    Module.Index0Param0Write=true
    Module.Index0Param1Write=true
    Module.Index0Param2Write=true
    Module.Index0Param3Write=true
   )
   &Color(Color=&MultiplyLinearColor(
     Linear Color=&ColorFromCurve(ColorCurve=DataInterface() {propertyValues {
        AlphaCurve {keys [<at 0 0>,<at 0.0834 1>,<at 0.8243 1 RCIM_Cubic>,<at 1 0 RCIM_Cubic>]}
        BlueCurve {keys [
          <at 0.0115 1.25 RCIM_Cubic>
          {arriveTangentWeight=0.209, interpMode=RCIM_Cubic, leaveTangentWeight=0.1865, tangentMode=RCTM_User, time=0.1245, value=0.2}
          {arriveTangent=-1.3308, interpMode=RCIM_Cubic, leaveTangent=-1.3308, time=0.6019, value=0}
          <at 0.9508 0 RCIM_Cubic>
         ]}
        GreenCurve {keys [
          <at 0.0115 1.25 RCIM_Cubic>
          {
           arriveTangent=-1.3308
           arriveTangentWeight=0.1735
           interpMode=RCIM_Cubic
           leaveTangent=-1.3308
           leaveTangentWeight=0.1548
           time=0.1245
           value=0.6
          }
          {arriveTangent=-1.3308, interpMode=RCIM_Cubic, leaveTangent=-1.3308, time=0.6019, value=0.25}
          <at 0.9508 0 RCIM_Cubic>
         ]}
        RedCurve {keys [
          <at 0.0115 1.25 RCIM_Cubic>
          {
           arriveTangent=-0.2662
           arriveTangentWeight=0.1125
           interpMode=RCIM_Cubic
           leaveTangent=-0.2662
           leaveTangentWeight=0.1004
           time=0.1245
           value=1
          }
          {arriveTangent=-0.2662, interpMode=RCIM_Cubic, leaveTangent=-0.2662, time=0.6019, value=1}
          <at 0.9508 1 RCIM_Cubic>
         ]}
       }})
     Scale Factor=&MakeFloat4FromVectorAndFloat(
      "Vector (RGB)"=&VectorFromFloat(Value=&RandomRangeFloat(Minimum=0.5, Maximum=2))
      "Float (Alpha)"=1
     )
    ))
  ]
  rendererClasses [&NiagaraSpriteRendererProperties]
  renderers [{rendererClass=&NiagaraSpriteRendererProperties}]
 }
 Small_rico {
  emitterUpdateScript [
   &EmitterState()
   &SpawnBurst_Instantaneous(Spawn Count=&RandomRangeInt(Minimum=10, Maximum=15), Use Spawn Probability=false)
  ]
  particleSpawnScript [
   &InitializeParticle(
    Color=&LerpLinearColors(StartColor=<rgba 1.5 0.4313 0.075 1>, EndColor=<rgba 3 1.8744 0.4518 1>, LerpFactor=$System.Impactintencity)
    Lifetime=&Lerp_Float(
     A=0.05
     B=0.15
     Alpha=&FloatFromCurve(
      FloatCurve=DataInterface() {propertyValues {Curve {keys [<at 0 0 RCIM_Cubic>, {arriveTangent=3.573, interpMode=RCIM_Cubic, leaveTangent=3.573, tangentMode=RCTM_User, time=1, value=1}]}}}
      CurveIndex=&RandomRangeFloat(Minimum=0, Maximum=1)
     )
    )
    Mass Max=3
    Mass Mode=::NewEnumerator3
    Sprite Size Mode=::NewEnumerator3
    Uniform Sprite Size=&RandomRangeFloat(Minimum=1.5, Maximum=2)
    UsePositionOffset=false
    Write Material Random=true
   ) {module=InitializeParticle001}
   &AddVelocityInCone(
    Cone Angle=5
    Cone Axis=&RotateVector(
     Vector To Rotate:Vector3f=&MakeVector(X=0.55, Y=&Multiply_Float(A=0.2, B=$System.ImpactSideAngleVariation), Z=0.9)
     Yaw:NiagaraFloat=0
     Pitch:NiagaraFloat=0
     Roll:NiagaraFloat=&Multiply_Float(A=10, B=&RandomRangeFloat(Minimum=-1, Maximum=1))
     Rotation Rate:NiagaraFloat=1
    )
    Module.Use Velocity Falloff On Cone Axis=false
    Velocity Strength=&Multiply_Float(A=&UniformRangedFloat(Minimum=150, Maximum=300), B=12)
   )
   &InheritVelocity(Inherited Velocity Speed Limit=2000, Speed Limit=true, UseSpeedThreshold=false) {enabled=false}
   &CameraOffset(Camera Offset Amount=-5)
  ]
  particleUpdateScript [
   &ParticleState()
   &InheritSourceMovement(
    Applied Position Scale=&VectorFromFloat(Value=&FloatFromCurve(FloatCurve=DataInterface() {propertyValues {Curve {keys [
          {arriveTangent=-3.3349, interpMode=RCIM_Cubic, leaveTangent=-3.3349, tangentMode=RCTM_User, time=0, value=1}
          {interpMode=RCIM_Cubic, tangentMode=RCTM_User, time=0.5, value=0}
         ]}}}))
    ApplyForce=false
    ApplyPosition=true
    ApplyPositionFalloff=false
   )
   &ScaleColor(
    Scale Alpha=&FloatFromCurve(FloatCurve=DataInterface() {propertyValues {Curve {keys [
         {tangentMode=RCTM_User, time=0, value=1}
         {arriveTangent=-0.0405, interpMode=RCIM_Cubic, leaveTangent=-0.0405, tangentMode=RCTM_User, time=0.7009, value=1}
         {arriveTangent=-8.0089, interpMode=RCIM_Cubic, leaveTangent=-8.0089, tangentMode=RCTM_User, time=1, value=0}
        ]}}})
    ScaleA=true
    ScaleRGB=true
   )
   &Drag(Drag=10, Use Linear Drag=true, Use Rotational Drag=true)
   &GravityForce(Gravity=<0 0 -9800>)
   &SolveForcesAndVelocity(Clamp Velocity=false, Limit Acceleration=false)
   &ScaleSpriteSizeBySpeed(
    Max Scale Factor=&RandomRangeVector2D(Minimum=<0.5 7>, Maximum=<0.5 10>)
    Min Scale Factor=<0.1 0.5>
    Velocity Threshold=1400
   )
   &DynamicMaterialParameters(
    Index 0 Param 1=1
    Index 0 Param 2=1
    Index 0 Param 3=&FloatFromCurve(FloatCurve=DataInterface() {propertyValues {Curve {keys [<at 0 1 RCIM_Cubic>,<at 1 0 RCIM_Cubic>]}}})
    Index 0 Param 4=1
    Module.Index0Param0Write=true
    Module.Index0Param1Write=true
    Module.Index0Param2Write=true
    Module.Index0Param3Write=true
   )
   &Color(Color=&MultiplyLinearColor(
     Linear Color=&ColorFromCurve(ColorCurve=DataInterface() {propertyValues {
        AlphaCurve {keys [<at 0.8243 1 RCIM_Cubic>,<at 1 0 RCIM_Cubic>]}
        BlueCurve {keys [
          <at 0.0115 1.25 RCIM_Cubic>
          {arriveTangentWeight=0.209, interpMode=RCIM_Cubic, leaveTangentWeight=0.1865, tangentMode=RCTM_User, time=0.1937, value=0.2}
          {arriveTangent=-0.2642, interpMode=RCIM_Cubic, leaveTangent=-0.2642, time=0.6501, value=0}
          <at 0.9508 0 RCIM_Cubic>
         ]}
        GreenCurve {keys [
          <at 0.0115 1.25 RCIM_Cubic>
          {
           arriveTangent=-1.5658
           arriveTangentWeight=0.1735
           interpMode=RCIM_Cubic
           leaveTangent=-1.5658
           leaveTangentWeight=0.1548
           time=0.1937
           value=0.6
          }
          {arriveTangent=-0.7925, interpMode=RCIM_Cubic, leaveTangent=-0.7925, time=0.6501, value=0.25}
          <at 0.9508 0 RCIM_Cubic>
         ]}
        RedCurve {keys [
          <at 0.0115 1.25 RCIM_Cubic>
          {
           arriveTangent=-0.3915
           arriveTangentWeight=0.1125
           interpMode=RCIM_Cubic
           leaveTangent=-0.3915
           leaveTangentWeight=0.1004
           time=0.1937
           value=1
          }
          <at 0.6501 1 RCIM_Cubic>
          <at 0.9508 1 RCIM_Cubic>
         ]}
       }})
     Scale Factor=&MakeFloat4FromVectorAndFloat(
      "Vector (RGB)"=&VectorFromFloat(Value=&RandomRangeFloat(Minimum=0.5, Maximum=2))
      "Float (Alpha)"=1
     )
    ))
  ]
  rendererClasses [&NiagaraSpriteRendererProperties]
  renderers [{rendererClass=&NiagaraSpriteRendererProperties}]
 }
}
summary {
 systemName=NS_Ashton_Shield_Block_Impact
 emitters [
  {emitterName=Rico_bullets, rendererClasses [&NiagaraSpriteRendererProperties]}
  {emitterName=Small_rico, rendererClasses [&NiagaraSpriteRendererProperties]}
 ]
 userVariables [User.ImpactStage {type=NiagaraFloat, defaultValue=0, description=""}]
}
