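# Niagara effect for Captain America's shield that flashes a short-lived impact mesh reflecting the shield's damage state.
# ShieldImpact bursts one randomly sized and rotated mesh particle, drawn by two mesh renderers with dynamic material parameters.
# User.DamageState indexes curves that shift color from pale yellow through orange to red while lowering alpha toward zero.
# Lifetime curves add a hot initial flash, alpha fade and slight growth, with alpha passed through a time-of-day scalar.
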NiagaraSystem /AshtonGameplay/Assets/CaptainAmerica/Shield/FX/NS_Ashton_Shield_DamageState
dependencies {
 usedDataInterfaces [&NiagaraDataInterfaceColorCurve, &NiagaraDataInterfaceCurve]
 usedDynamicInputs [&RandomRangeVector]
 usedModules [
  &SystemState
  &self:SystemSpawnScript.SystemScriptSource.SystemScriptGraph.NiagaraNodeAssignment_0.SetVariables_#1
 ]
}
emitters {ShieldImpact {
  emitterUpdateScript [
   &EmitterState()
   &SpawnBurst_Instantaneous(Spawn Count=1, Use Spawn Probability=false)
  ]
  particleSpawnScript [
   /SetVariables(Particles.ShieldSize:NiagaraFloat=&RandomRangeFloat(Minimum=1.25, Maximum=1.3)) {isSetParameters=true}
   &InitializeParticle(
    Lifetime=0.4
    Mesh Scale=&MakeVector(X=$Particles.ShieldSize, Y=$Particles.ShieldSize, Z=1)
    Mesh Scale Mode=::NewEnumerator4
    UsePositionOffset=false
    Write Material Random=true
   )
   &InitialMeshOrientation(
    Orientation Axis=<0 0 1>
    Orientation Vector=<-1 0 0>
    Rotation=&MakeVector(X=0, Y=0, Z=&RandomRangeFloat(Minimum=-1, Maximum=1))
    Use Orientation Vector=true
    Use Rotation Vector=true
   )
   &SystemLocation(MaskSpawnGroup=false, Offset=<-6 -3 -0.5>)
   /SetVariables(Particles.MeshOrientationAlt:Quat4f=&MultiplyQuaternion(
     Quaternion A=$Particles.MeshOrientation
     Quaternion B=&MakeQuatFromAxisAngle(Axis=<0 0 1>, AngleInDegrees=&RandomRangeFloat(Minimum=-360, Maximum=360))
    )) {isSetParameters=true}
  ]
  particleUpdateScript [
   &ParticleState()
   &DynamicMaterialParameters(
    Index 0 Param 1=&RandomRangeFloat(Minimum=0.1, Maximum=0.2)
    Index 0 Param 2=&RandomRangeFloat(Minimum=0.2, Maximum=0.3)
    Index 0 Param 3=0.5
    Index 0 Param 4=0.324
    Module.Param0WriteEnabled=true
    Module.Param1WriteEnabled=true
    Module.Param2WriteEnabled=true
    Module.Param3WriteEnabled=true
   )
   &ScaleColor(
    Scale Alpha=&Multiply_Float(
     A=&Lerp_Float(
      A=25
      B=0
      Alpha=&FloatFromCurve(
       FloatCurve=DataInterface() {propertyValues {Curve {keys [
           {arriveTangent=0.2474, interpMode=RCIM_Cubic, leaveTangent=0.2474, tangentMode=RCTM_User, time=0, value=0}
           {arriveTangent=1.3152, interpMode=RCIM_Cubic, leaveTangent=1.3152, tangentMode=RCTM_User, time=1, value=1}
          ]}}}
       CurveIndex=$User.DamageState
      )
     )
     B=&FloatFromCurve(
      FloatCurve=DataInterface() {propertyValues {Curve {keys [
          {interpMode=RCIM_Cubic, tangentMode=RCTM_User, time=0, value=0}
          <at 0.1 1 RCIM_Cubic>
          {interpMode=RCIM_Cubic, tangentMode=RCTM_User, time=1, value=0}
         ]}}}
      Scale Curve=&FNTimeOfDayScalar(Night Scale=1, Morning Scale=1, Evening Scale=1)
     )
    )
    Scale RGB=&MakeVectorFromLinearColorRGB(Color=&ColorFromCurve(
      ColorCurve=DataInterface() {propertyValues {
        AlphaCurve {keys [<at 0 1 RCIM_Cubic>]}
        BlueCurve {keys [<at 0 0.55>,<at 0.2138 0.15>,<at 0.5261 0.036>,<at 0.8561 0>]}
        GreenCurve {keys [<at 0 0.8125>,<at 0.2138 0.6458>,<at 0.5261 0.2088>,<at 0.8561 0>]}
        RedCurve {keys [<at 0 1>,<at 0.2138 1>,<at 0.5261 0.9>,<at 0.8561 1>]}
       }}
      CurveIndex=$User.DamageState
     ))
    ScaleA=true
    ScaleRGB=true
   )
   &ScaleColor(
    Scale Alpha=&Multiply_Float(
     A=&Lerp_Float(
      A=4
      B=0
      Alpha=&FloatFromCurve(
       FloatCurve=DataInterface() {propertyValues {Curve {keys [
           {arriveTangent=0.2474, interpMode=RCIM_Cubic, leaveTangent=0.2474, tangentMode=RCTM_User, time=0, value=0}
           {interpMode=RCIM_Cubic, leaveTangent=0.0606, tangentMode=RCTM_User, time=1, value=1}
          ]}}}
       CurveIndex=$User.DamageState
      )
     )
     B=&FloatFromCurve(
      FloatCurve=DataInterface() {propertyValues {Curve {keys [
          {interpMode=RCIM_Cubic, tangentMode=RCTM_User, time=0, value=0}
          <at 0.3 1 RCIM_Cubic>
          {interpMode=RCIM_Cubic, tangentMode=RCTM_User, time=1, value=0}
         ]}}}
      Scale Curve=&FNTimeOfDayScalar(Night Scale=0.7, Morning Scale=0.8, Evening Scale=0.8)
     )
    )
    Scale RGB=&MakeVectorFromLinearColorRGB(Color=&ColorFromCurve(
      ColorCurve=DataInterface() {propertyValues {
        AlphaCurve {keys [<at 0 1 RCIM_Cubic>]}
        BlueCurve {keys [
          <at 0 2>
          {arriveTangent=-3.4008, interpMode=RCIM_Cubic, leaveTangent=-3.4008, time=0.2127, value=2}
          <at 0.3562 0.098>
          <at 0.5593 0.098>
          {arriveTangent=-0.2224, interpMode=RCIM_Cubic, leaveTangent=-0.2224, time=0.7606, value=0}
          <at 1 0>
         ]}
        GreenCurve {keys [
          <at 0 0.8>
          {arriveTangent=-0.6991, interpMode=RCIM_Cubic, leaveTangent=-0.6991, time=0.2127, value=0.8}
          <at 0.3562 0.409>
          <at 0.5593 0.409>
          {arriveTangent=-0.9281, interpMode=RCIM_Cubic, leaveTangent=-0.9281, time=0.7606, value=0}
          <at 1 0>
         ]}
        RedCurve {keys [
          <at 0 0.2>
          {arriveTangent=0.894, interpMode=RCIM_Cubic, leaveTangent=0.894, time=0.2127, value=0.2}
          <at 0.3562 0.7>
          <at 0.5593 0.7>
          {arriveTangent=0.6807, interpMode=RCIM_Cubic, leaveTangent=0.6807, time=0.7606, value=1}
          <at 1 1>
         ]}
       }}
      CurveIndex=$User.DamageState
     ))
    ScaleA=true
    ScaleRGB=true
   ) {enabled=false, module=ScaleColor001}
   &ScaleMeshSize(Scale Factor=&VectorFromFloat(Value=&FloatFromCurve(FloatCurve=DataInterface() {propertyValues {Curve {keys [
          {arriveTangent=0.5961, interpMode=RCIM_Cubic, leaveTangent=0.5961, tangentMode=RCTM_User, time=0, value=1}
          {interpMode=RCIM_Cubic, tangentMode=RCTM_User, time=1, value=1.2}
         ]}}}))) {module=ScaleMeshSize001}
   &ScaleMeshSize(Scale Factor=&MakeVector(
     X=&Lerp_Float(
      A=0.8
      B=1
      Alpha=&FloatFromCurve(FloatCurve=DataInterface() {propertyValues {Curve {keys [
           {interpMode=RCIM_Cubic, tangentMode=RCTM_User, time=0, value=0}
           {arriveTangentWeight=0.5062, interpMode=RCIM_Cubic, tangentMode=RCTM_User, time=0.9, value=1}
          ]}}})
     )
     Y=&Lerp_Float(
      A=0.8
      B=1
      Alpha=&FloatFromCurve(FloatCurve=DataInterface() {propertyValues {Curve {keys [
           {interpMode=RCIM_Cubic, tangentMode=RCTM_User, time=0, value=0}
           {arriveTangentWeight=0.5062, interpMode=RCIM_Cubic, tangentMode=RCTM_User, time=0.9, value=1}
          ]}}})
     )
     Z=1
    )) {enabled=false}
   &ScaleColor(
    Scale RGB=&MakeVectorFromLinearColorRGB(Color=&ColorFromCurve(ColorCurve=DataInterface() {propertyValues {
        AlphaCurve {keys [<at 0 1 RCIM_Cubic>]}
        BlueCurve {keys [<at 0 35 RCIM_Cubic>,<at 0.192 2>,<at 0.8081 0.5 RCIM_Cubic>]}
        GreenCurve {keys [<at 0 42.5 RCIM_Cubic>,<at 0.192 2>,<at 0.8081 0.5 RCIM_Cubic>]}
        RedCurve {keys [<at 0 50 RCIM_Cubic>,<at 0.192 2>,<at 0.8081 1 RCIM_Cubic>]}
       }}))
    ScaleA=false
    ScaleRGB=true
   ) {module=ScaleColor002}
  ]
  rendererClasses [&NiagaraMeshRendererProperties]
  renderers [
   {rendererClass=&NiagaraMeshRendererProperties, rendererIndex=0}
   {rendererClass=&NiagaraMeshRendererProperties, rendererIndex=1}
  ]
 }}
summary {
 systemName=NS_Ashton_Shield_DamageState
 emitters [{emitterName=ShieldImpact, rendererClasses [&NiagaraMeshRendererProperties]}]
 userVariables [User.DamageState {type=NiagaraFloat, defaultValue=0, description=""}]
}
