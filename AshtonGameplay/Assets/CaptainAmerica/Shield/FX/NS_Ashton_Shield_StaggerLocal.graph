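# Niagara one-shot effect for Captain America's shield stagger, built from two instant-burst mesh emitters at a fixed position offset.
# Shield_Shards001 selects shard meshes by seven-way partition, grows them, cools HDR red to black and animates dynamic material parameters.
# Its outward velocity and scripted position modules are disabled, so motion comes from scale and material.
# Wind_Cone adds a short, pale-red cone that expands while its alpha flashes in and out.
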NiagaraSystem /AshtonGameplay/Assets/CaptainAmerica/Shield/FX/NS_Ashton_Shield_StaggerLocal
dependencies {
 usedDataInterfaces [
  &NiagaraDataInterfaceCurve
  &NiagaraDataInterfaceMeshRendererInfo
  &NiagaraDataInterfaceArrayFloat3
  &NiagaraDataInterfaceColorCurve
 ]
 usedDynamicInputs []
 usedModules [&SystemState]
}
emitters {
 Shield_Shards001 {
  emitterUpdateScript [
   &EmitterState()
   &SpawnBurst_Instantaneous(Use Spawn Probability=false, UseLoopCountLimit=false)
  ]
  particleSpawnScript [
   &PartitionParticles(ID to Partition:NiagaraInt32=&ReturnExecIndex(), Number Of Partitions=7)
   &InitializeParticle(
    Color=<rgba 1 1 1 4749.8232>
    Lifetime=&RandomRangeFloat(Minimum=0.225, Maximum=0.36)
    Mesh Renderer Array Visibility Mode=::NewEnumerator1
    Mesh Renderer Info=DataInterface() {propertyValues {MeshRenderer=&self:Shield_Shards001_1.NiagaraMeshRendererProperties_1}}
    Mesh Renderer Mesh Index:NiagaraInt32=$Particles.PartitionParticles.Partition
    Mesh Scale Mode=::NewEnumerator3
    Position Offset=<0 50 100>
    UsePositionOffset=true
    Write Material Random=true
   )
   &AddVelocity(Velocity=&Multiply_VectorByFloat(
     Vector=&SelectVectorFromArray(
      Direct Array Index=$Particles.PartitionParticles.Partition
      Vector Selection Array=DataInterface() {propertyValues {FloatData [<-1.1 0.2 -1.1>, <-1 0.2 0.3>, <0 0.2 1>, <1 0.2 1>, <1 0.2 0>, <1.1 0.2 -0.5>, <0.3 0 -1.5>]}}
     )
     Float=&RandomRangeFloat(Minimum=300, Maximum=400)
    )) {enabled=false}
   &InitialMeshOrientation(Orientation Vector=<-1 0 0>, Rotation=<0.25 0 0>, Use Orientation Vector=true, Use Rotation Vector=true)
  ]
  particleUpdateScript [
   &ParticleState()
   &ScaleMeshSize(Scale Factor=&VectorFromFloat(Value=&FloatFromCurve(FloatCurve=DataInterface() {propertyValues {Curve {keys [{arriveTangent=0.5857, interpMode=RCIM_Cubic, leaveTangent=0.5857, tangentMode=RCTM_User, time=0, value=0.75}, <at 1 1.25 RCIM_Cubic>]}}})))
   &UpdateMeshOrientation(
    Facing Direction=&RotateVectorByQuaternion(
     Vector To Rotate=<1 0 0>
     Quaternion=&MakeQuatFromAxisAngle(Axis=<0 1 0>, AngleInDegrees=&Multiply_Float(A=$System.SystemRotation, B=360))
    )
    Facing Interpolation=false
    Orientation Method=::NewEnumerator1
    Override Quat=false
    TransformFacing=false
    TransformUp=false
    Up Direction:Vector3f=<1 0 0>
   ) {enabled=false}
   &ScaleColor(
    Scale Alpha=&FloatFromCurve(FloatCurve=DataInterface() {propertyValues {Curve {keys [{arriveTangent=0.1248, interpMode=RCIM_Cubic, leaveTangent=0.1248, tangentMode=RCTM_User, time=0, value=1}, <at 1 0>]}}})
    Scale RGB=&MakeVectorFromLinearColorRGB(Color=&ColorFromCurve(ColorCurve=DataInterface() {propertyValues {
        AlphaCurve {keys [<at 0 1 RCIM_Cubic>]}
        BlueCurve {keys [<at 0.1466 0.2 RCIM_Cubic>, <at 0.3335 0.1>, {interpMode=RCIM_Cubic, leaveTangent=-0.7465, time=1, value=0}]}
        GreenCurve {keys [<at 0.1466 0.2 RCIM_Cubic>, <at 0.3335 0.1>, {interpMode=RCIM_Cubic, leaveTangent=-1.1489, time=1, value=0}]}
        RedCurve {keys [<at 0.1466 200 RCIM_Cubic>, <at 0.3335 10>, {interpMode=RCIM_Cubic, leaveTangent=-24.8851, time=1, value=0}]}
       }}))
    ScaleA=true
    ScaleRGB=true
   )
   /SetVariables(Particles.Position=&AddVectorToPosition(
     Position=<0 50 100>
     Vector=&Multiply_VectorByFloat(
      Vector=&SelectVectorFromArray(
       Direct Array Index=$Particles.PartitionParticles.Partition
       Vector Selection Array=DataInterface() {propertyValues {FloatData [<-1 0 0.5>, <1 0 0.5>, <0 0 -1>]}}
      )
      Float=&Multiply_Float(
       A=30
       B=&FloatFromCurve(FloatCurve=DataInterface() {propertyValues {Curve {keys [
            {arriveTangent=2.59, interpMode=RCIM_Cubic, leaveTangent=2.59, tangentMode=RCTM_User, time=0, value=0}
            {interpMode=RCIM_Cubic, tangentMode=RCTM_User, time=0.9, value=1}
           ]}}})
      )
     )
    )) {enabled=false, isSetParameters=true}
   &Drag(
    Drag=&FloatFromCurve(
     FloatCurve=DataInterface() {propertyValues {Curve {keys [
         {arriveTangent=4.3219, interpMode=RCIM_Cubic, tangentMode=RCTM_User, time=0, value=0}
         {interpMode=RCIM_Cubic, tangentMode=RCTM_User, time=0.15, value=1}
        ]}}}
     Scale Curve=&RandomRangeFloat(Minimum=10, Maximum=14)
    )
    Use Linear Drag=true
    Use Rotational Drag=true
   )
   &SolveForcesAndVelocity(Clamp Velocity=false, Limit Acceleration=false)
   &DynamicMaterialParameters(
    Index 0 Param 1=&FloatFromCurve(
     FloatCurve=DataInterface() {propertyValues {Curve {keys [
         {arriveTangent=0.8495, interpMode=RCIM_Cubic, leaveTangent=0.8495, tangentMode=RCTM_User, time=0.1, value=0.2}
         {interpMode=RCIM_Cubic, tangentMode=RCTM_User, time=0.9, value=1}
        ]}}}
     Scale Curve=-1
    )
    Index 0 Param 2=1
    Index 0 Param 3=&FloatFromCurve(FloatCurve=DataInterface() {propertyValues {Curve {keys [{interpMode=RCIM_Cubic, tangentMode=RCTM_User, time=0.1, value=0}, {interpMode=RCIM_Cubic, tangentMode=RCTM_User, time=0.9, value=1}]}}})
    Module.Index0Param0Write=true
    Module.Index0Param1Write=true
    Module.Index0Param2Write=true
    Module.Index0Param3Write=false
   )
  ]
  rendererClasses [&NiagaraMeshRendererProperties]
  renderers [{rendererClass=&NiagaraMeshRendererProperties, rendererIndex=0}]
 }
 Wind_Cone {
  emitterUpdateScript [
   &EmitterState()
   &SpawnBurst_Instantaneous(Use Spawn Probability=false, UseLoopCountLimit=false)
  ]
  particleSpawnScript [
   &InitializeParticle(
    Color=<rgba 1 0.5417 0.5 250>
    Lifetime=0.2
    Mesh Scale=<0.5 1.5 1.5>
    Mesh Scale Mode=::NewEnumerator4
    Position Offset=<0 50 100>
    UsePositionOffset=true
    Write Material Random=true
   )
   &InitialMeshOrientation(Mesh Orientation Mode=::NewEnumerator2, Orientation Vector=<0 1 0>, Use Rotation Vector=true)
  ]
  particleUpdateScript [
   &ParticleState()
   &ScaleMeshSize(Scale Factor=&VectorFromFloat(Value=&FloatFromCurve(FloatCurve=DataInterface() {propertyValues {Curve {keys [{tangentMode=RCTM_User, time=0, value=0}, {tangentMode=RCTM_User, time=1, value=1}]}}})))
   &ScaleColor(
    Scale Alpha=&FloatFromCurve(
     FloatCurve=DataInterface() {propertyValues {Curve {keys [
         {interpMode=RCIM_Cubic, tangentMode=RCTM_User, time=0, value=0}
         {interpMode=RCIM_Cubic, tangentMode=RCTM_User, time=0.35, value=1}
         {tangentMode=RCTM_User, time=1, value=0}
        ]}}}
     Scale Curve=15
    )
    Scale RGB=&VectorFromFloat(Value=1)
    ScaleA=true
    ScaleRGB=true
   )
  ]
  rendererClasses [&NiagaraMeshRendererProperties]
  renderers [{rendererClass=&NiagaraMeshRendererProperties, rendererIndex=0}]
 }
}
summary {
 systemName=NS_Ashton_Shield_StaggerLocal
 emitters [
  {emitterName=Wind_Cone, rendererClasses [&NiagaraMeshRendererProperties]}
  {emitterName=Shield_Shards001, rendererClasses [&NiagaraMeshRendererProperties]}
 ]
 userVariables []
}
