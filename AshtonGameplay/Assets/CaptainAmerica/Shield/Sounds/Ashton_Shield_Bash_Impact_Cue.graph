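# Sound cue for the impact of a Captain America shield bash in Ashton gameplay.
# A Random node picks one of four bash recordings with equal weight, then a Modulator node feeds the output.
# The cue plays at reduced volume through the third-person enemy player foley sound class.
# It references a shield attenuation asset with attenuation override enabled, and overrides concurrency to cap simultaneous instances.
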SoundCue /AshtonGameplay/Assets/CaptainAmerica/Shield/Sounds/Ashton_Shield_Bash_Impact_Cue
root=Modulator_2
nodes [
 {id=Modulator_2, children [Random_2]}
 {id=Random_2, children [WavePlayer_6,WavePlayer_7,WavePlayer_8,WavePlayer_9], properties {weights [1,1,1,1]}}
 {id=WavePlayer_6, properties {soundWaveAssetPtr=&/Ashton_Shield_Bash_04}}
 {id=WavePlayer_7, properties {soundWaveAssetPtr=&/Ashton_Shield_Bash_03}}
 {id=WavePlayer_8, properties {soundWaveAssetPtr=&/Ashton_Shield_Bash_02}}
 {id=WavePlayer_9, properties {soundWaveAssetPtr=&/Ashton_Shield_Bash_01}}
]
properties {
 attenuationSettings=&/Ashton_Shield_Att
 bOverrideAttenuation=true
 bOverrideConcurrency=true
 maxDistance=4000
 soundClassObject=&Fort_Player_Foley_3P_Enemy_Athena
 volumeMultiplier=0.7
 concurrencyOverrides {maxCount=3}
}
