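# Sound Cue for Captain America shield charge footsteps, in the melee weapon sound class and concurrency set.
# An islocal branch plays 2D attenuation when true, and volume-boosted regular shield attenuation when false or unset.
# Both share a mixer of randomized Husk steps and pitched-down Constructor metal steps, with the shield-specific heavy-footstep layer at zero volume.
# A quality-level switch shrinks the Husk pool from nine waves to three for its last two slots.
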SoundCue /AshtonGameplay/Assets/CaptainAmerica/Shield/Sounds/Ashton_Shield_Charge_Footsteps_Cue
root=Branch_0
nodes [
 {
  id=Attenuation_0
  children [Mixer_1]
  properties {attenuationSettings=&/Ashton_Shield_2D_Att}
 }
 {id=Attenuation_1, children [Mixer_1], properties {attenuationSettings=&/Ashton_Shield_Att}}
 {id=Branch_0, children [Attenuation_0, Modulator_0, Modulator_0], properties {boolParameterName=islocal}}
 {id=Mixer_1, children [Modulator_5,Modulator_3,Modulator_1], properties {inputVolume [0.3,0.45,0]}}
 {id=Modulator_0, children [Attenuation_1], properties {pitchMax=1, volumeMax=1.2, volumeMin=1.2}}
 {id=Modulator_1, children [Random_2]}
 {id=Modulator_3, children [Random_1], properties {pitchMax=0.75, pitchMin=0.8}}
 {id=Modulator_5, children [QualityLevel_0]}
 {id=QualityLevel_0, children [Random_6,Random_6,Random_5,Random_5]}
 {id=Random_1, children [WavePlayer_0,WavePlayer_1,WavePlayer_2,WavePlayer_3], properties {weights [1,1,1,1]}}
 {id=Random_2, children [WavePlayer_36, WavePlayer_37, WavePlayer_38, WavePlayer_39], properties {weights [1,1,1,1]}}
 {id=Random_5, children [WavePlayer_53, WavePlayer_54, WavePlayer_55], properties {bShouldExcludeFromBranchCulling=true, weights [1,1,1]}}
 {
  id=Random_6
  children [WavePlayer_43, WavePlayer_44, WavePlayer_45, WavePlayer_46, WavePlayer_47, WavePlayer_52, WavePlayer_53, WavePlayer_54, WavePlayer_55]
  properties {weights [1,1,1,1,1,1,1,1,1]}
 }
 {id=WavePlayer_0, properties {soundWaveAssetPtr=/Game/Sounds/Fort_Step_Sounds/Constructor/Constructor_Step_Walk_Metal_10}}
 {id=WavePlayer_1, properties {soundWaveAssetPtr=/Game/Sounds/Fort_Step_Sounds/Constructor/Constructor_Step_Walk_Metal_11}}
 {id=WavePlayer_2, properties {soundWaveAssetPtr=/Game/Sounds/Fort_Step_Sounds/Constructor/Constructor_Step_Walk_Metal_12}}
 {id=WavePlayer_3, properties {soundWaveAssetPtr=/Game/Sounds/Fort_Step_Sounds/Constructor/Constructor_Step_Walk_Metal_01}}
 {id=WavePlayer_36, properties {soundWaveAssetPtr=&/Ashton_Shield_HeavyFootsteps_17}}
 {id=WavePlayer_37, properties {soundWaveAssetPtr=&/Ashton_Shield_HeavyFootsteps_16}}
 {id=WavePlayer_38, properties {soundWaveAssetPtr=&/Ashton_Shield_HeavyFootsteps_03}}
 {id=WavePlayer_39, properties {soundWaveAssetPtr=&/Ashton_Shield_HeavyFootsteps_01}}
 {id=WavePlayer_43, properties {soundWaveAssetPtr=/Game/Sounds/Fort_Step_Sounds/Husk/husky_step_01}}
 {id=WavePlayer_44, properties {soundWaveAssetPtr=/Game/Sounds/Fort_Step_Sounds/Husk/husky_step_02}}
 {id=WavePlayer_45, properties {soundWaveAssetPtr=/Game/Sounds/Fort_Step_Sounds/Husk/husky_step_03}}
 {id=WavePlayer_46, properties {soundWaveAssetPtr=/Game/Sounds/Fort_Step_Sounds/Husk/husky_step_04}}
 {id=WavePlayer_47, properties {soundWaveAssetPtr=/Game/Sounds/Fort_Step_Sounds/Husk/husky_step_05}}
 {id=WavePlayer_52, properties {soundWaveAssetPtr=/Game/Sounds/Fort_Step_Sounds/Husk/husky_step_06}}
 {id=WavePlayer_53, properties {soundWaveAssetPtr=/Game/Sounds/Fort_Step_Sounds/Husk/husky_step_07}}
 {id=WavePlayer_54, properties {soundWaveAssetPtr=/Game/Sounds/Fort_Step_Sounds/Husk/husky_step_08}}
 {id=WavePlayer_55, properties {soundWaveAssetPtr=/Game/Sounds/Fort_Step_Sounds/Husk/husky_step_09}}
]
properties {
 maxDistance=3400
 soundClassObject=&Fort_WeaponMelee
 volumeMultiplier=0.6
 attenuationOverrides {distanceAlgorithm=NaturalSound, falloffDistance=3000, attenuationShapeExtents {x=1000}}
 concurrencySet [&Player_Melee_Weapon_Conc]
}
