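# Sound cue for the intro of Captain America's shield charge, played as a melee weapon sound with shield attenuation and limited to one concurrent instance.
# A top mixer layers a random whoosh, a quieter random start sound and a nested mixer.
# The nested mixer combines a fixed charge-end sound with a random "empty" variant that is briefly delayed and pitched up.
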SoundCue /AshtonGameplay/Assets/CaptainAmerica/Shield/Sounds/Ashton_Shield_Charge_Intro_Cue
root=Mixer_0
nodes [
 {id=Delay_0, children [Modulator_2], properties {delayMax=0.1, delayMin=0.1}}
 {id=Mixer_0, children [Random_0,Random_2,Mixer_1], properties {inputVolume [0.8,0.25,1]}}
 {id=Mixer_1, children [Delay_0,WavePlayer_6], properties {inputVolume [0.8,1.1]}}
 {id=Modulator_2, children [Random_1], properties {pitchMax=1.2, pitchMin=1.2}}
 {id=Random_0, children [WavePlayer_0, WavePlayer_9, WavePlayer_10, WavePlayer_11], properties {weights [1,1,1,1]}}
 {id=Random_1, children [WavePlayer_3,WavePlayer_4,WavePlayer_5], properties {weights [1,1,1]}}
 {id=Random_2, children [WavePlayer_1,WavePlayer_2], properties {weights [1,1]}}
 {id=WavePlayer_0, properties {soundWaveAssetPtr=&/Ashton_Shield_Charge_Whoosh_04}}
 {
  id=WavePlayer_1
  properties {soundWaveAssetPtr=/AshtonGameplay/Assets/CaptainAmerica/Shield/Sounds/_version/Ashton_Shield_Charge_Start_03v01}
 }
 {id=WavePlayer_10, properties {soundWaveAssetPtr=&/Ashton_Shield_Charge_Whoosh_02}}
 {id=WavePlayer_11, properties {soundWaveAssetPtr=&/Ashton_Shield_Charge_Whoosh_01}}
 {id=WavePlayer_2, properties {soundWaveAssetPtr=/AshtonGameplay/Assets/CaptainAmerica/Shield/Sounds/_version/Ashton_Shield_Charge_Start_02v1}}
 {id=WavePlayer_3, properties {soundWaveAssetPtr=&/Ashton_Shield_Charge_Empty_01}}
 {id=WavePlayer_4, properties {soundWaveAssetPtr=&/Ashton_Shield_Charge_Empty_02}}
 {id=WavePlayer_5, properties {soundWaveAssetPtr=&/Ashton_Shield_Charge_Empty_03}}
 {id=WavePlayer_6, properties {soundWaveAssetPtr=/AshtonGameplay/Assets/CaptainAmerica/Shield/Sounds/_version/Ashton_Shield_Charge_End}}
 {id=WavePlayer_9, properties {soundWaveAssetPtr=&/Ashton_Shield_Charge_Whoosh_03}}
]
properties {
 attenuationSettings=&/Ashton_Shield_Att
 bOverrideConcurrency=true
 maxDistance=3400
 soundClassObject=&Fort_WeaponMelee
 volumeMultiplier=0.95
 concurrencyOverrides {maxCount=1}
}
