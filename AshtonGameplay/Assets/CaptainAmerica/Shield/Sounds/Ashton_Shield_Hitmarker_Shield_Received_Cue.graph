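# Sound cue for the hitmarker heard when a hit on shield is received, built from Captain America shield impact waves.
# A quality switch sends the first two levels to a random pick of twelve impacts and the last two to a three-wave subset exempt from branch culling.
# A neutral modulator feeds the output, using hit-notification attenuation and sound class, with concurrency limiting playback to two instances by stopping the oldest.
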SoundCue /AshtonGameplay/Assets/CaptainAmerica/Shield/Sounds/Ashton_Shield_Hitmarker_Shield_Received_Cue
root=Modulator_1
nodes [
 {id=Modulator_1, children [QualityLevel_0], properties {pitchMax=1, pitchMin=1, volumeMax=1, volumeMin=1}}
 {id=QualityLevel_0, children [Random_0,Random_0,Random_2,Random_2]}
 {
  id=Random_0
  children [
   WavePlayer_0
   WavePlayer_1
   WavePlayer_2
   WavePlayer_3
   WavePlayer_6
   WavePlayer_7
   WavePlayer_8
   WavePlayer_9
   WavePlayer_10
   WavePlayer_11
   WavePlayer_12
   WavePlayer_13
  ]
  properties {weights [1,1,1,1,1,1,1,1,1,1,1,1]}
 }
 {id=Random_2, children [WavePlayer_7, WavePlayer_10, WavePlayer_13], properties {bShouldExcludeFromBranchCulling=true, weights [1,1,1]}}
 {id=WavePlayer_0, properties {soundWaveAssetPtr=&/Ashton_Shield_Impact_12}}
 {id=WavePlayer_1, properties {soundWaveAssetPtr=&/Ashton_Shield_Impact_11}}
 {id=WavePlayer_10, properties {soundWaveAssetPtr=&/Ashton_Shield_Impact_04}}
 {id=WavePlayer_11, properties {soundWaveAssetPtr=&/Ashton_Shield_Impact_03}}
 {id=WavePlayer_12, properties {soundWaveAssetPtr=&/Ashton_Shield_Impact_02}}
 {id=WavePlayer_13, properties {soundWaveAssetPtr=&/Ashton_Shield_Impact_01}}
 {id=WavePlayer_2, properties {soundWaveAssetPtr=&/Ashton_Shield_Impact_10}}
 {id=WavePlayer_3, properties {soundWaveAssetPtr=&/Ashton_Shield_Impact_09}}
 {id=WavePlayer_6, properties {soundWaveAssetPtr=&/Ashton_Shield_Impact_08}}
 {id=WavePlayer_7, properties {soundWaveAssetPtr=&/Ashton_Shield_Impact_07}}
 {id=WavePlayer_8, properties {soundWaveAssetPtr=&/Ashton_Shield_Impact_06}}
 {id=WavePlayer_9, properties {soundWaveAssetPtr=&/Ashton_Shield_Impact_05}}
]
properties {
 attenuationSettings=/Game/Athena/Sounds/HitNotifications/Fort_HitMarker_Receive_Att
 maxDistance=2097152
 soundClassObject=/Game/Sounds/FortSoundClassesAndModes/Fort_2DHitNotifications_Receive_Shield
 volumeMultiplier=0.8
 concurrencyOverrides {maxCount=2, resolutionRule=StopOldest}
 concurrencySet [&HitNotifications_TwoMaximum_StopOldest_VolumeScale]
}
