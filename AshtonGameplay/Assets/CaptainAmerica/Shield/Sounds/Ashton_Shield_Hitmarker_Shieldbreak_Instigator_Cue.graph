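# Sound cue for the instigator's hitmarker when Captain America's shield (Ashton) breaks a shield.
# Mixes three layers: a pitched-down Boombox metal impact faded out by a volume envelope, a randomly picked shield stagger sound at boosted volume, and a randomly picked glass-break sound.
# Plays through the 2D instigator hit-notification sound class and attenuation, with concurrency capped at two voices, stopping the oldest.
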SoundCue /AshtonGameplay/Assets/CaptainAmerica/Shield/Sounds/Ashton_Shield_Hitmarker_Shieldbreak_Instigator_Cue
root=Modulator_1
nodes [
 {
  id=Enveloper_0
  children [WavePlayer_1]
  properties {loopCount=1, pitchCurve {editorCurveData {keys [{time=0}]}}, volumeCurve {editorCurveData {keys [<at 0 1>,<at 0.5 0>]}}}
 }
 {id=Mixer_0, children [Modulator_0,Modulator_2,Modulator_3], properties {inputVolume [0.4,1.1,0.8]}}
 {id=Modulator_0, children [Enveloper_0], properties {pitchMax=0.75, pitchMin=0.75}}
 {id=Modulator_1, children [Mixer_0], properties {volumeMax=1, volumeMin=1}}
 {id=Modulator_2, children [Random_0], properties {pitchMax=1, pitchMin=1, volumeMax=2, volumeMin=2}}
 {id=Modulator_3, children [Random_1]}
 {id=Random_0, children [WavePlayer_0,WavePlayer_2,WavePlayer_3,WavePlayer_5], properties {weights [1,1,1,1]}}
 {id=Random_1, children [WavePlayer_4,WavePlayer_6,WavePlayer_7,WavePlayer_8], properties {weights [1,1,1,1]}}
 {id=WavePlayer_0, properties {soundWaveAssetPtr=&/Ashton_Shield_Stagger_In_01}}
 {id=WavePlayer_1, properties {soundWaveAssetPtr=/Game/Sounds/Fort_Weapons/Boombox/Club/Boombox_Club_Swing_Impact_Metal_v1_01}}
 {id=WavePlayer_2, properties {soundWaveAssetPtr=&/Ashton_Shield_Stagger_In_02}}
 {id=WavePlayer_3, properties {soundWaveAssetPtr=&/Ashton_Shield_Stagger_In_03}}
 {id=WavePlayer_4, properties {soundWaveAssetPtr=&/Ashton_Shield_ShieldBreakGlass_01}}
 {id=WavePlayer_5, properties {soundWaveAssetPtr=&/Ashton_Shield_Stagger_In_04}}
 {id=WavePlayer_6, properties {soundWaveAssetPtr=&/Ashton_Shield_ShieldBreakGlass_02}}
 {id=WavePlayer_7, properties {soundWaveAssetPtr=&/Ashton_Shield_ShieldBreakGlass_03}}
 {id=WavePlayer_8, properties {soundWaveAssetPtr=&/Ashton_Shield_ShieldBreakGlass_04}}
]
properties {
 attenuationSettings=/Game/Athena/Sounds/HitNotifications/Fort_HitMarker_Instigate_Att
 maxDistance=2097152
 soundClassObject=&Fort_2DHitNotifications_Instigate_PSM
 volumeMultiplier=0.7
 concurrencyOverrides {maxCount=2, resolutionRule=StopOldest}
 concurrencySet [&HitNotifications_TwoMaximum_StopOldest_VolumeScale]
}
