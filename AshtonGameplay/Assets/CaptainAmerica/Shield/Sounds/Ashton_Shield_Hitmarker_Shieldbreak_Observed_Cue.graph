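# Sound cue for the shield-break hitmarker of Captain America's shield (Ashton), in its observed variant with hitmarker attenuation.
# A mixer layers a pitched-down metal impact faded by a volume envelope, a boosted random stagger sound, and a briefly delayed random glass-break sound.
# It plays in the impacts sound class, with concurrency capped at two and stop-oldest resolution.
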SoundCue /AshtonGameplay/Assets/CaptainAmerica/Shield/Sounds/Ashton_Shield_Hitmarker_Shieldbreak_Observed_Cue
root=Modulator_0
nodes [
 {id=Delay_0, children [Modulator_5], properties {delayMax=0.1, delayMin=0.1}}
 {
  id=Enveloper_0
  children [WavePlayer_1]
  properties {loopCount=1, pitchCurve {editorCurveData {keys [{time=0}]}}, volumeCurve {editorCurveData {keys [<at 0 1>,<at 0.5 0>]}}}
 }
 {id=Mixer_0, children [Modulator_3,Modulator_4,Delay_0], properties {inputVolume [0.5,3,1.5]}}
 {id=Modulator_0, children [Mixer_0], properties {volumeMax=1, volumeMin=1}}
 {id=Modulator_3, children [Enveloper_0], properties {pitchMax=0.75, pitchMin=0.75}}
 {id=Modulator_4, children [Random_1], properties {pitchMax=1, pitchMin=1, volumeMax=2, volumeMin=2}}
 {id=Modulator_5, children [Random_2]}
 {id=Random_1, children [WavePlayer_0,WavePlayer_7,WavePlayer_8,WavePlayer_9], properties {weights [1,1,1,1]}}
 {id=Random_2, children [WavePlayer_10, WavePlayer_11, WavePlayer_12, WavePlayer_13], properties {weights [1,1,1,1]}}
 {id=WavePlayer_0, properties {soundWaveAssetPtr=&/Ashton_Shield_Stagger_In_01}}
 {id=WavePlayer_1, properties {soundWaveAssetPtr=/Game/Sounds/Fort_Weapons/Boombox/Club/Boombox_Club_Swing_Impact_Metal_v1_01}}
 {id=WavePlayer_10, properties {soundWaveAssetPtr=&/Ashton_Shield_ShieldBreakGlass_01}}
 {id=WavePlayer_11, properties {soundWaveAssetPtr=&/Ashton_Shield_ShieldBreakGlass_02}}
 {id=WavePlayer_12, properties {soundWaveAssetPtr=&/Ashton_Shield_ShieldBreakGlass_03}}
 {id=WavePlayer_13, properties {soundWaveAssetPtr=&/Ashton_Shield_ShieldBreakGlass_04}}
 {id=WavePlayer_7, properties {soundWaveAssetPtr=&/Ashton_Shield_Stagger_In_02}}
 {id=WavePlayer_8, properties {soundWaveAssetPtr=&/Ashton_Shield_Stagger_In_03}}
 {id=WavePlayer_9, properties {soundWaveAssetPtr=&/Ashton_Shield_Stagger_In_04}}
]
properties {
 attenuationSettings=/Game/Athena/Sounds/HitNotifications/Fort_HitMarker_Observed_Att
 maxDistance=5500
 soundClassObject=&Fort_Impacts
 volumeMultiplier=0.45
 concurrencyOverrides {maxCount=2, resolutionRule=StopOldest}
 concurrencySet [&HitNotifications_TwoMaximum_StopOldest_VolumeScale]
}
