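# Sound cue for lowering the Captain America shield, built from ADS-out waves and routed to the melee weapon sound class and player melee concurrency.
# A quality-level node sends its first two slots to an eight-wave random pool and the last two to a three-wave subset.
# A top modulator randomizes pitch well upward, partly offset by the cue's lower pitch multiplier.
# It references weapon-reload attenuation, and the debug flag is left on.
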SoundCue /AshtonGameplay/Assets/CaptainAmerica/Shield/Sounds/Ashton_Shield_Lower_Cue
root=Modulator_0
nodes [
 {id=Modulator_0, children [QualityLevel_0], properties {pitchMax=2.5, pitchMin=2.25}}
 {id=QualityLevel_0, children [Random_4,Random_4,Random_0,Random_0]}
 {id=Random_0, children [WavePlayer_5, WavePlayer_6, WavePlayer_10], properties {bShouldExcludeFromBranchCulling=true, weights [1,1,1]}}
 {
  id=Random_4
  children [WavePlayer_0, WavePlayer_1, WavePlayer_2, WavePlayer_3, WavePlayer_4, WavePlayer_5, WavePlayer_6, WavePlayer_10]
  properties {bShouldExcludeFromBranchCulling=true, weights [1,1,1,1,1,1,1,1]}
 }
 {id=WavePlayer_0, properties {soundWaveAssetPtr=&/Ashton_Shield_ADS_Out_01}}
 {id=WavePlayer_1, properties {soundWaveAssetPtr=&/Ashton_Shield_ADS_Out_02}}
 {id=WavePlayer_10, properties {soundWaveAssetPtr=&/Ashton_Shield_ADS_Out_08}}
 {id=WavePlayer_2, properties {soundWaveAssetPtr=&/Ashton_Shield_ADS_Out_03}}
 {id=WavePlayer_3, properties {soundWaveAssetPtr=&/Ashton_Shield_ADS_Out_04}}
 {id=WavePlayer_4, properties {soundWaveAssetPtr=&/Ashton_Shield_ADS_Out_05}}
 {id=WavePlayer_5, properties {soundWaveAssetPtr=&/Ashton_Shield_ADS_Out_06}}
 {id=WavePlayer_6, properties {soundWaveAssetPtr=&/Ashton_Shield_ADS_Out_07}}
]
properties {
 attenuationSettings=&Fort_Weapon_Reload_Att
 bDebug=true
 maxDistance=2250
 pitchMultiplier=0.82
 soundClassObject=&Fort_WeaponMelee
 volumeMultiplier=0.6
 attenuationOverrides {distanceAlgorithm=NaturalSound, falloffDistance=3000, attenuationShapeExtents {x=1000}}
 concurrencySet [&Player_Melee_Weapon_Conc]
}
