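# Sound cue for picking up Captain America's shield, classed as loot collect with weapon pickup attenuation and a pickup concurrency set.
# A mixer layers randomized shield foley, two throw foley sets and pickup sounds, with the pickup layer loudest.
# Modulators vary the shield foley and pickup layers, whose quality-level switches use smaller random pools for the last two levels.
# The smaller pickup pool includes an empty slot.
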SoundCue /AshtonGameplay/Assets/CaptainAmerica/Shield/Sounds/Ashton_Shield_Pickup_Cue
root=Mixer_0
nodes [
 {id=Mixer_0, children [Modulator_0,Random_6,Random_5,Modulator_1], properties {inputVolume [1.5,1.5,1,2]}}
 {id=Modulator_0, children [QualityLevel_0]}
 {id=Modulator_1, children [QualityLevel_1]}
 {id=QualityLevel_0, children [Random_0,Random_0,Random_1,Random_1]}
 {id=QualityLevel_1, children [Random_2,Random_2,Random_3,Random_3]}
 {id=Random_0, children [WavePlayer_0,WavePlayer_1,WavePlayer_2,WavePlayer_3,WavePlayer_4], properties {weights [1,1,1,1,1]}}
 {id=Random_1, children [WavePlayer_2,WavePlayer_3,WavePlayer_4], properties {bShouldExcludeFromBranchCulling=true, weights [1,1,1]}}
 {id=Random_2, children [WavePlayer_5, WavePlayer_6, WavePlayer_8, WavePlayer_10], properties {weights [1,1,1,1]}}
 {id=Random_3, children [null, WavePlayer_8, WavePlayer_10], properties {bShouldExcludeFromBranchCulling=true, weights [1,1,1]}}
 {id=Random_5, children [WavePlayer_13, WavePlayer_14, WavePlayer_15, WavePlayer_16], properties {weights [1,1,1,1]}}
 {id=Random_6, children [WavePlayer_9, WavePlayer_11, WavePlayer_12], properties {weights [1,1,1]}}
 {id=WavePlayer_0, properties {soundWaveAssetPtr=&/Ashton_Shield_Foley_05}}
 {id=WavePlayer_1, properties {soundWaveAssetPtr=&/Ashton_Shield_Foley_04}}
 {id=WavePlayer_10, properties {soundWaveAssetPtr=&/Ashton_Shield_PickUp_01}}
 {id=WavePlayer_11, properties {soundWaveAssetPtr=/AshtonGameplay/Assets/CaptainAmerica/Shield/Sounds/_version/Ashton_Shield_Throw_FoleyA_02}}
 {id=WavePlayer_12, properties {soundWaveAssetPtr=/AshtonGameplay/Assets/CaptainAmerica/Shield/Sounds/_version/Ashton_Shield_Throw_FoleyA_03}}
 {id=WavePlayer_13, properties {soundWaveAssetPtr=/AshtonGameplay/Assets/CaptainAmerica/Shield/Sounds/_version/Ashton_Shield_Throw_FoleyB_01}}
 {id=WavePlayer_14, properties {soundWaveAssetPtr=/AshtonGameplay/Assets/CaptainAmerica/Shield/Sounds/_version/Ashton_Shield_Throw_FoleyB_02}}
 {id=WavePlayer_15, properties {soundWaveAssetPtr=/AshtonGameplay/Assets/CaptainAmerica/Shield/Sounds/_version/Ashton_Shield_Throw_FoleyB_03}}
 {id=WavePlayer_16, properties {soundWaveAssetPtr=/AshtonGameplay/Assets/CaptainAmerica/Shield/Sounds/_version/Ashton_Shield_Throw_FoleyB_04}}
 {id=WavePlayer_2, properties {soundWaveAssetPtr=&/Ashton_Shield_Foley_03}}
 {id=WavePlayer_3, properties {soundWaveAssetPtr=&/Ashton_Shield_Foley_02}}
 {id=WavePlayer_4, properties {soundWaveAssetPtr=&/Ashton_Shield_Foley_01}}
 {id=WavePlayer_5, properties {soundWaveAssetPtr=&/Ashton_Shield_PickUp_05}}
 {id=WavePlayer_6, properties {soundWaveAssetPtr=&/Ashton_Shield_PickUp_04}}
 {id=WavePlayer_8, properties {soundWaveAssetPtr=&/Ashton_Shield_PickUp_02}}
 {id=WavePlayer_9, properties {soundWaveAssetPtr=/AshtonGameplay/Assets/CaptainAmerica/Shield/Sounds/_version/Ashton_Shield_Throw_FoleyA_01}}
]
properties {
 attenuationSettings=&Fort_Weapon_Pickup_Att
 maxDistance=3500
 soundClassObject=&Fort_Loot_Collect
 volumeMultiplier=0.45
 concurrencyOverrides {maxCount=1, resolutionRule=PreventNew}
 concurrencySet [&PickupMaximum_Oldest]
}
