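# Sound cue for raising Captain America's shield, playing one of its ADS-in sound variations.
# A Modulator feeds a quality-level switch: the first two levels randomize evenly across eight waves, the other two across three of those same waves.
# The cue is pitched up and lowered in volume, with melee weapon sound class and concurrency and a shield attenuation asset.
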SoundCue /AshtonGameplay/Assets/CaptainAmerica/Shield/Sounds/Ashton_Shield_Raise_Cue
root=Modulator_0
nodes [
 {id=Modulator_0, children [QualityLevel_0]}
 {id=QualityLevel_0, children [Random_0,Random_0,Random_1,Random_1]}
 {
  id=Random_0
  children [WavePlayer_0, WavePlayer_1, WavePlayer_2, WavePlayer_3, WavePlayer_4, WavePlayer_5, WavePlayer_6, WavePlayer_10]
  properties {weights [1,1,1,1,1,1,1,1]}
 }
 {id=Random_1, children [WavePlayer_5, WavePlayer_6, WavePlayer_10], properties {bShouldExcludeFromBranchCulling=true, weights [1,1,1]}}
 {id=WavePlayer_0, properties {soundWaveAssetPtr=&/Ashton_Shield_ADS_In_01}}
 {id=WavePlayer_1, properties {soundWaveAssetPtr=&/Ashton_Shield_ADS_In_02}}
 {id=WavePlayer_10, properties {soundWaveAssetPtr=&/Ashton_Shield_ADS_In_08}}
 {id=WavePlayer_2, properties {soundWaveAssetPtr=&/Ashton_Shield_ADS_In_03}}
 {id=WavePlayer_3, properties {soundWaveAssetPtr=&/Ashton_Shield_ADS_In_04}}
 {id=WavePlayer_4, properties {soundWaveAssetPtr=&/Ashton_Shield_ADS_In_05}}
 {id=WavePlayer_5, properties {soundWaveAssetPtr=&/Ashton_Shield_ADS_In_06}}
 {id=WavePlayer_6, properties {soundWaveAssetPtr=&/Ashton_Shield_ADS_In_07}}
]
properties {
 attenuationSettings=&/Ashton_Shield_Att
 maxDistance=3400
 pitchMultiplier=1.25
 soundClassObject=&Fort_WeaponMelee
 volumeMultiplier=0.35
 attenuationOverrides {distanceAlgorithm=NaturalSound, falloffDistance=3000, attenuationShapeExtents {x=1000}}
 concurrencySet [&Player_Melee_Weapon_Conc]
}
