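# Sound cue for the end of a stagger with the Captain America shield, going by its name, routed to the melee weapon sound class.
# A short fixed delay feeds an equally weighted random pick among four max-charge one-shot waves, reused here at reduced volume.
# It uses the shared shield attenuation asset and overrides concurrency to cap simultaneous instances at three.
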SoundCue /AshtonGameplay/Assets/CaptainAmerica/Shield/Sounds/Ashton_Shield_StaggerEnd_Cue
root=Delay_0
nodes [
 {id=Delay_0, children [Random_4], properties {delayMax=0.1, delayMin=0.1}}
 {id=Random_4, children [WavePlayer_1,WavePlayer_2,WavePlayer_3,WavePlayer_4], properties {weights [1,1,1,1]}}
 {id=WavePlayer_1, properties {soundWaveAssetPtr=&/Ashton_Shield_MaxChargeOneShot_01}}
 {id=WavePlayer_2, properties {soundWaveAssetPtr=&/Ashton_Shield_MaxChargeOneShot_02}}
 {id=WavePlayer_3, properties {soundWaveAssetPtr=&/Ashton_Shield_MaxChargeOneShot_03}}
 {id=WavePlayer_4, properties {soundWaveAssetPtr=&/Ashton_Shield_MaxChargeOneShot_04}}
]
properties {
 attenuationSettings=&/Ashton_Shield_Att
 bOverrideConcurrency=true
 maxDistance=3400
 soundClassObject=&Fort_WeaponMelee
 volumeMultiplier=0.25
 concurrencyOverrides {maxCount=3}
}
